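# Invisible sub-category carousel button widget for the Locker, derived from the UIKit generic modular button base and adding no widgets or variables of its own.
# UpdateState takes a HasNewItems flag and passes it to Set Bang State on the status indicator, with a zero count, to toggle the new-items bang.
# PreConstruct and Construct only forward to the parent, and Tick is empty.
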WidgetBlueprint /Locker/MPL/WBP_SubCategoryCarousel_Button_invisible
graphs [
 EventGraph {nodes [
   n1 = Event PreConstruct()
   > Parent: PreConstruct(IsDesignTime=@n1.IsDesignTime)
   Event Construct()
   > Parent: Construct()
   Event Tick()
  ]}
 UpdateState {nodes [
   n1 = UpdateState!() -> (then, HasNewItems:bool=~) {is=title K2Node_FunctionEntry}
   > Set Bang State!(
    execute
    self:Fort Bang Wrapper NUI Object Reference=Get StatusIndicator() -> (StatusIndicator:WBP UIKit Status Indicator Object Reference)
    bEnabled:bool=@n1
    Count:int=0
   ) -> (then) {is=title K2Node_CallFunction}
  ]}
]
variables []
widgets {
 info {
  inheritedWidgetCount=0
  parentClass=/FortUILibrary/Systems/ModularButton/WBP_UIKit_ButtonGeneric_Base.WBP_UIKit_ButtonGeneric_Base_C
  rootWidgetClass=null
  widgetCount=0
 }
 widgets []
}
